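# Sound cue for the activation sound of the Holo Telescope quest actor in a Season 25 Battle Pass story quest (Part 03, Q04).
# A single wave player plays the HoloTelescope activate wave through a Modulator node at the root.
# It routes to the Fort_Quests sound class and uses the shared quest attenuation asset for spatial falloff.
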SoundCue /BattlePassS25/Items/QuestActors/Stories/Part03/SFX/S25_P03_Q04_HoloTelescope_Activate_Cue
root=Modulator_0
nodes [
 {id=Modulator_0, children [WavePlayer_0]}
 {id=WavePlayer_0, properties {soundWaveAssetPtr=&/S25_P03_Q04_HoloTelescope_Activate}}
]
properties {
 attenuationSettings=&/S25_P03_Q04_Atten
 maxDistance=1100
 soundClassObject=&Fort_Quests
}
